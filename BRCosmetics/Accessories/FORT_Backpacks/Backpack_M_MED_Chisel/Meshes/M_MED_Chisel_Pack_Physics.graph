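# Physics asset for the M_MED_Chisel backpack, defining a small humanoid figure of dyn_ bones attached to the pack.
# Capsules cover head, torso, arms and legs, while the pack bone carries a small capsule plus a large, offset box.
# Upper body and hips join the pack as separate chains: dyn_body swings and twists freely, dyn_hip is tightly limited.
# Limb and head joints have angular limits, with very narrow twist throughout.
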PhysicsAsset /BRCosmetics/Accessories/FORT_Backpacks/Backpack_M_MED_Chisel/Meshes/M_MED_Chisel_Pack_Physics
bodies {
 dyn_arm_lft [{shapeType=Capsule, length=2.875, radius=1.625, shapeName=dyn_arm_lft_capsule, center=<2.5 0 -0.0797>, rotation=<rot -90 0 0>}]
 dyn_arm_rgt [{shapeType=Capsule, length=3.125, radius=1.875, shapeName=dyn_arm_rgt_capsule, center=<2.5 0 0.1572>, rotation=<rot -90 0 0>}]
 dyn_body [{shapeType=Capsule, length=1.625, radius=4.375, shapeName=dyn_body_capsule, center=<-5 0.1082 -1.5844>, rotation=<rot 90 0 0>}]
 dyn_foot_lft [{shapeType=Capsule, length=5.5, radius=3.5, shapeName=dyn_foot_lft_capsule, center=<3.5 0.5 2>, rotation=<rot 10 -90 0>}]
 dyn_foot_rgt [{shapeType=Capsule, length=5.5, radius=3.5, shapeName=dyn_foot_rgt_capsule, center=<3.5 -0.5 2>, rotation=<rot -10 -90 0>}]
 dyn_hand_lft [{
   shapeType=Capsule
   length=3.625
   radius=2.375
   shapeName=dyn_hand_lft_capsule
   center=<3.7942 0.2913 -0.64>
   rotation=<rot -79.9998 180 -179.9997>
  }]
 dyn_hand_rgt [{shapeType=Capsule, length=3.375, radius=2.625, shapeName=dyn_hand_rgt_capsule, center=<3.8597 -0.0761 -0.8129>, rotation=<rot -70 180 180>}]
 dyn_head [{shapeType=Capsule, length=12.125, radius=6.125, shapeName=dyn_head_capsule, center=<5.6541 0 0>, rotation=<rot 0.0004 0 -90>}]
 dyn_hip [{shapeType=Capsule, length=3.625, radius=3.875, shapeName=dyn_hip_capsule, center=<0 0.6956 0>}]
 dyn_knee_lft [{shapeType=Capsule, length=2.25, radius=2.7, shapeName=dyn_knee_lft_capsule, center=<1 -0.2 -0.5>, rotation=<rot 70 90 90>}]
 dyn_knee_rgt [{shapeType=Capsule, length=2.25, radius=2.75, shapeName=dyn_knee_rgt_capsule, center=<1 0.3 -0.5>, rotation=<rot -25 -90 -90>}]
 dyn_leg_lft [{shapeType=Capsule, length=2, radius=2.75, shapeName=dyn_leg_lft_capsule, center=<2 -0.5 -0.2>, rotation=<rot 15 -90 -90>}]
 dyn_leg_rgt [{shapeType=Capsule, length=1.75, radius=2.75, shapeName=dyn_leg_rgt_capsule, center=<2 0.5 -0.2>, rotation=<rot -5 90 -90>}]
 dyn_shoulder_lft [{shapeType=Capsule, length=2.375, radius=1.875, shapeName=dyn_shoulder_lft_capsule, center=<2 0 0.5>, rotation=<rot -80 179.9974 -179.9964>}]
 dyn_shoulder_rgt [{shapeType=Capsule, length=3.125, radius=1.875, shapeName=dyn_shoulder_rgt_capsule, center=<2 0 0.5>, rotation=<rot -80 -179.9984 179.9977>}]
 pack [
  {shapeType=Capsule, length=0.75, radius=0.75, shapeName=pack_capsule}
  {shapeType=Box, extentX=37, extentY=6, extentZ=30, shapeName=pack_box, center=<0 5.8631 -10>}
 ]
}
constraints [
 {
  bone1Name=dyn_body
  bone2Name=pack
  swing1LimitDegrees=45
  swing1Motion=Free
  swing2LimitDegrees=45
  swing2Motion=Free
  twistLimitDegrees=45
  twistMotion=Free
 }
 {bone1Name=dyn_head, bone2Name=dyn_body, swing1LimitDegrees=15, swing2LimitDegrees=10, twistLimitDegrees=2}
 {bone1Name=dyn_shoulder_lft, bone2Name=dyn_body, swing1LimitDegrees=45, swing2LimitDegrees=20, twistLimitDegrees=2}
 {bone1Name=dyn_arm_lft, bone2Name=dyn_shoulder_lft, swing1LimitDegrees=30, swing2LimitDegrees=20, twistLimitDegrees=2}
 {bone1Name=dyn_hand_lft, bone2Name=dyn_arm_lft, swing1LimitDegrees=45, swing2LimitDegrees=20, twistLimitDegrees=2}
 {bone1Name=dyn_shoulder_rgt, bone2Name=dyn_body, swing1LimitDegrees=45, swing2LimitDegrees=20, twistLimitDegrees=2}
 {bone1Name=dyn_arm_rgt, bone2Name=dyn_shoulder_rgt, swing1LimitDegrees=30, swing2LimitDegrees=20, twistLimitDegrees=2}
 {bone1Name=dyn_hand_rgt, bone2Name=dyn_arm_rgt, swing1LimitDegrees=45, swing2LimitDegrees=20, twistLimitDegrees=2}
 {bone1Name=dyn_hip, bone2Name=pack, swing1LimitDegrees=10, swing2LimitDegrees=10, twistLimitDegrees=5}
 {bone1Name=dyn_leg_lft, bone2Name=dyn_hip, swing1LimitDegrees=10, swing2LimitDegrees=45, twistLimitDegrees=2}
 {bone1Name=dyn_knee_lft, bone2Name=dyn_leg_lft, swing1LimitDegrees=10, swing2LimitDegrees=30, twistLimitDegrees=2}
 {bone1Name=dyn_foot_lft, bone2Name=dyn_knee_lft, swing1LimitDegrees=5, swing2LimitDegrees=5, twistLimitDegrees=2}
 {bone1Name=dyn_leg_rgt, bone2Name=dyn_hip, swing1LimitDegrees=10, swing2LimitDegrees=45, twistLimitDegrees=2}
 {bone1Name=dyn_knee_rgt, bone2Name=dyn_leg_rgt, swing1LimitDegrees=10, swing2LimitDegrees=30, twistLimitDegrees=2}
 {bone1Name=dyn_foot_rgt, bone2Name=dyn_knee_rgt, swing1LimitDegrees=5, swing2LimitDegrees=5, twistLimitDegrees=2}
]
